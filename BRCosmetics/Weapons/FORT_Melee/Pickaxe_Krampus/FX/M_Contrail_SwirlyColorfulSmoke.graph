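# Translucent unlit ribbon and beam trail material for swirly, colorful smoke contrails on the Krampus pickaxe.
# A panning swirl flowmap distorts scrolling smoke, which blends three-color gradients along U, with a two-color switch fallback.
# Emissive adds swirl brightness, time-of-day scaling, oversaturation via negative desaturation, and particle color.
# Opacity combines smoke, trail mask, sphere-mask end fades, time of day, depth fade and particle/MasterAlpha, with zero on mobile.
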Material /BRCosmetics/Weapons/FORT_Melee/Pickaxe_Krampus/FX/M_Contrail_SwirlyColorfulSmoke
graph {
 properties {
  bUsedWithBeamTrails=true
  bUsedWithNiagaraRibbons=true
  blendMode=BLEND_Translucent
  pixelDepthOffsetMode=PDOM_Legacy
  shadingModel=MSM_Unlit
  translucencyLightingMode=TLM_VolumetricPerVertexDirectional
  usedShadingModels=Unlit
 }
 nodes [
  n1 = MaterialFunctionCall(Float2=TextureCoordinate()) {materialFunction=&BreakOutFloat2Components}
  n2 = MaterialFunctionCall(Float2=TextureCoordinate()) {materialFunction=&BreakOutFloat2Components}
  n3 = ScalarParameter() {defaultValue=0.85}
  n4 = ComponentMask(None=Multiply(A=TextureCoordinate(), B=ScalarParameter() {defaultValue=0.02, parameterName=U_Scale})) {r=true}
  n5 = AppendVector(A=Multiply(A=@n4) {constB=21}, B=ComponentMask(None=TextureCoordinate()) {g=true})
  n6 = Panner(Coordinate=TextureCoordinate() {uTiling=0.52, vTiling=0.52}) {speedX=-0.75}
  n7 = Multiply(A=ComponentMask(None=TextureSample(UVs=@n6) {samplerType=SAMPLERTYPE_Normal, texture=/Game/Global/Textures/Default/Flowmaps/T_SwirlFlow_02_N}.RGB) {g=true, r=true}) {constB=0.11}
  n8 = Multiply(A=DynamicParameter() {paramNames [Param1,Param2,Param3,MasterAlpha]}.Param3, B=Time())
  n9 = TextureSample(UVs=Panner(Coordinate=Add(A=@n5, B=@n7), Time=@n8) {speedX=2}) {samplerType=SAMPLERTYPE_Grayscale, texture=&T_Smoke_Trail_Soft}
  n10 = LinearInterpolate(
   A=LinearInterpolate(
    A=VectorParameter() {parameterName=ColorC, defaultValue=<rgba 0 0.9688 0.8727 1>}.RGB
    B=VectorParameter() {parameterName=ColorA, defaultValue=<rgba 0.1276 0.3217 0.7379 1>}.RGB
    Alpha=Multiply(A=@n1.R, B=@n1.R)
   )
   B=LinearInterpolate(
    A=VectorParameter() {parameterName=ColorB, defaultValue=<rgba 1 0.052 0.062 1>}.RGB
    B=VectorParameter() {parameterName=ColorA, defaultValue=<rgba 0.1276 0.3217 0.7379 1>}.RGB
    Alpha=Power(Base=@n2.R, Exp=@n3)
   )
   Alpha=Power(Base=@n9.RGB, Exp=ScalarParameter() {defaultValue=0.85, parameterName=Contrast})
  )
  n11 = LinearInterpolate(
   A=Constant3Vector() {constant=<rgba 0.1276 0.3217 0.7379 1>}
   B=Constant3Vector() {constant=<rgba 0.8646 0.0097 0.2996 1>}
   Alpha=Power(Base=@n4, Exp=ScalarParameter() {defaultValue=1.8, parameterName=Color Contrast})
  )
  n12 = Constant() {r=0.04}
  n13 = Constant() {r=1.35}
  n14 = AppendVector(A=Multiply(A=@n4) {constB=21.5}, B=ComponentMask(None=TextureCoordinate()) {g=true})
  n15 = Power(Base=@n9.R, Exp=ScalarParameter() {defaultValue=1.21, parameterName=Alpha Scope}) {constExponent=1.3}
  n16 = Add(
   A=TextureSample(UVs=Panner(Coordinate=@n14) {speedX=1.5}) {samplerType=SAMPLERTYPE_LinearColor, texture=/Game/Global/Textures/Default/Flowmaps/T_SwirlFlow_02}.R
   B=Multiply(A=@n15) {constB=0.65}
  )
  n17 = ScalarParameter() {defaultValue=1.45, parameterName=Param_4}
  n18 = ScalarParameter() {defaultValue=1, parameterName=Param_5}
  n19 = Multiply(
   A=Multiply(
    A=StaticSwitchParameter(True=@n10, False=@n11) {defaultValue=true, parameterName="WorldPositionColor "}
    B=LinearInterpolate(A=@n12, B=@n13, Alpha=@n16) {constA=0.5}
   )
   B=MaterialFunctionCall(Color_Morning=@n17, Color_Day=@n17, Color_Evening=@n18, Color_Night=@n18) {materialFunction=&Color_Time_of_Day}.Result
  ) {constB=2.4}
  n20 = ScalarParameter() {defaultValue=-1.75, parameterName=Saturation}
  n21 = ParticleColor()
  MaterialFunctionCall(Emissive=Multiply(A=Desaturation(None=@n19, Fraction=@n20), B=@n21.RGB)) {materialFunction=&MF_FX_EmissiveAdjustments}
  n22 = Panner(Coordinate=@n5, Time=@n8) {speedX=2}
  n23 = Multiply(A=Add(A=@n15, B=TextureSample(UVs=@n22) {texture=/Game/Effects/Fort_Effects/Textures/Smoke/T_Lobber_Trail_Mask}.RGB)) {constB=1.15}
  n24 = MaterialFunctionCall(Float2=TextureCoordinate()) {materialFunction=&BreakOutFloat2Components}
  n25 = Multiply(
   A=SphereMask(A=Power(Base=@n24.R) {constExponent=0.22}, B=Constant() {r=0.5}) {attenuationRadius=0.5, hardnessPercent=0}
   B=SphereMask(A=TextureCoordinate(), B=Constant2Vector() {g=0.5, r=1.1}) {attenuationRadius=1.05, hardnessPercent=88}
  )
  n26 = Constant() {r=2.1}
  n27 = Constant() {r=1}
  n28 = ShadingPathSwitch(
   Default=Multiply(
    A=Multiply(A=@n23, B=@n25)
    B=MaterialFunctionCall(Color_Morning=@n26, Color_Day=@n26, Color_Evening=@n27, Color_Night=@n27) {materialFunction=&Color_Time_of_Day}.Result
   )
   Mobile=Constant()
  ) {desc=Migrated from FeatureLevelSwitch}
  n29 = ScalarParameter() {defaultValue=100, parameterName=FadeDistance}
  Multiply(
   A=Multiply(A=@n21.A, B=DepthFade(Opacity=@n28, FadeDistance=@n29))
   B=DynamicParameter() {paramNames [Param1,Param2,Param3,MasterAlpha]}.MasterAlpha
  )
 ]
}
