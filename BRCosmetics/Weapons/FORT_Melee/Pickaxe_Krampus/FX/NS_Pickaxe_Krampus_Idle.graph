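# Niagara idle effect for the Krampus pickaxe: blue-tinted smoke wisps with a persistent blue light that fades in over the first second.
# Smoke_Local spawns two groups on opposite sides, switching offset, color and lifetime by SpawnGroup, which is copied to VisibilityTag alongside sprite and mesh renderers.
# Smoke_World adds short-lived flipbook sprites spawned per distance moved, inheriting part of the velocity to form a motion trail.
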NiagaraSystem /BRCosmetics/Weapons/FORT_Melee/Pickaxe_Krampus/FX/NS_Pickaxe_Krampus_Idle
dependencies {
 usedDataInterfaces [
  &NiagaraDataInterfaceCurve
  &NiagaraDataInterfaceVectorCurve
  &NiagaraDataInterfaceSpriteRendererInfo
 ]
 usedDynamicInputs []
 usedModules [&SystemState]
}
emitters {
 Light {
  emitterUpdateScript [
   &EmitterState(
    Inactive Response=::NewEnumerator1
    Life Cycle Mode=::NewEnumerator1
    Loop Behavior=::NewEnumerator1
    Loop Duration Mode=::NewEnumerator1
    UseLoopDelay=false
   )
   &SpawnBurst_Instantaneous(Use Spawn Probability=false, UseLoopCountLimit=false)
  ]
  particleSpawnScript [&InitializeParticle(UsePositionOffset=false, Write Material Random=true)]
  particleUpdateScript [
   &ParticleState(Kill Particles When Lifetime Has Elapsed=false)
   &FN_Light_Attributes(
    Color=&Multiply_LinearColorByFloat(
     Linear Color=<rgba 0.137 0.693 1.5 1>
     Float=&Multiply_Float(
      A=4
      B=&FloatFromCurve(FloatCurve=DataInterface() {propertyValues {Curve {keys [<at 0 0 RCIM_Cubic>,<at 1 1 RCIM_Cubic>]}}}, CurveIndex=$System.Age)
     )
    )
    Create Color Attribute=true
    Create Exponent Attribute=true
    Create Position Attribute=true
    Create Radius Attribute=true
    Position=<13 16 75>
    Radius=90
   )
  ]
  rendererClasses [&NiagaraLightRendererProperties]
  renderers [{rendererClass=&NiagaraLightRendererProperties, rendererIndex=0}]
 }
 Smoke_Local {
  emitterUpdateScript [
   &EmitterState()
   &SpawnRate(SpawnRate=3, Use Spawn Probability=false)
   &SpawnRate(SpawnGroup=1, SpawnRate=&RandomRangeFloat(Minimum=1, Maximum=2), Use Spawn Probability=false) {module=SpawnRate001}
  ]
  particleSpawnScript [
   &InitializeParticle(
    Color=&MakeLinearColorFromVectorAndFloat(
     "Vector (RGB)"=&MergeHelper_VectorSwitch(0=<3 3 3>, 1=<0.1 0.1 0.1>, 2=?, 3=?, Index=$Emitter.SpawnGroup, Num=::NewEnumerator0)
     "Float (Alpha)"=&MergeHelper_FloatSwitch(
      0=&RandomRangeFloat(Minimum=0.6, Maximum=1)
      1=3
      2=?
      3=?
      Index=$Emitter.SpawnGroup
      Num=::NewEnumerator0
     )
    )
    Lifetime=&MergeHelper_FloatSwitch(
     0=&RandomRangeFloat(Minimum=2.6, Maximum=3, Evaluation Type=::NewEnumerator1)
     1=1.5
     2=?
     3=?
     Index=$Emitter.SpawnGroup
     Num=::NewEnumerator0
    )
    Mesh Scale Mode=::NewEnumerator3
    Mesh Uniform Scale=0.75
    Position Offset=&MergeHelper_VectorSwitch(
     0=&RandomRangeVector(Minimum=<0 12 55>, Maximum=<0 22 55>, Evaluation Type=::NewEnumerator1)
     1=<0 -12 55>
     2=?
     3=?
     Index=$Emitter.SpawnGroup
     Num=::NewEnumerator0
    )
    Position Offset Coordinate Space=::Local
    Sprite Rotation Mode=::NewEnumerator2
    Sprite Size=&Vector2DFromFloat(Value=&RandomRangeFloat(Minimum=40, Maximum=50, Evaluation Type=::NewEnumerator1))
    Sprite Size Mode=::NewEnumerator4
    UsePositionOffset=true
    Write Material Random=true
   )
   /SetVariables(Particles.Position 02:NiagaraPosition=$Particles.Position) {isSetParameters=true}
   &ShapeLocation(
    Cylinder Height=20
    Cylinder Radius=5
    MaskSpawnGroup=false
    Override Seed=false
    Rotation Mode=::NewEnumerator2
    Shape Primitive=::NewEnumerator1
    Surface Only=false
    UseOwnerScale=true
    Yaw / Pitch / Roll=<0 0 0.25>
   )
   &DynamicMaterialParameters(
    Index 0 Param 1=&RandomRangeFloat(Minimum=0, Maximum=1, Evaluation Type=::NewEnumerator1)
    Index 0 Param 2=&RandomRangeFloat(Minimum=1, Maximum=2, Evaluation Type=::NewEnumerator1)
    Index 0 Param 3=0.1
    Index 0 Param 4=0.5
    Module.Index0Float0Write=true
    Module.Index0Float1Write=true
    Module.Index0Float2Write=true
    Module.Index0Float3Write=true
   )
   &AddVelocity(Velocity=&RandomRangeVector(Minimum=<0 -40 0>, Maximum=<0 -30 0>, Evaluation Type=::NewEnumerator1))
   /SetVariables(Particles.VisibilityTag=$Emitter.SpawnGroup) {isSetParameters=true}
  ]
  particleUpdateScript [
   &ParticleState()
   &ScaleMeshSize(Scale Factor=&VectorFromFloat(Value=&FloatFromCurve(FloatCurve=DataInterface() {propertyValues {Curve {keys [{arriveTangent=0.9501, interpMode=RCIM_Cubic, leaveTangent=0.9501, tangentMode=RCTM_User, time=0, value=0.4}, <at 1 0.9 RCIM_Cubic>]}}})))
   &ScaleSpriteSize(
    Scale Sprite Size Mode=::NewEnumerator0
    Uniform Scale Factor=&FloatFromCurve(FloatCurve=DataInterface() {propertyValues {Curve {keys [{arriveTangent=0.3852, time=0, value=0.4456}, <at 0.9031 1.3>]}}})
   )
   &ScaleColor(
    Scale Alpha=&MergeHelper_FloatSwitch(
     0=&FloatFromCurve(FloatCurve=DataInterface() {propertyValues {Curve {keys [
          <at 0 0 RCIM_Cubic>
          {arriveTangent=0.3168, interpMode=RCIM_Cubic, leaveTangent=0.3168, time=0.3375, value=1}
          {arriveTangent=-1.5094, interpMode=RCIM_Cubic, leaveTangent=-1.5094, time=0.7386, value=0.234}
          <at 1 0 RCIM_Cubic>
         ]}}})
     1=&FloatFromCurve(FloatCurve=DataInterface() {propertyValues {Curve {keys [<at 0 0 RCIM_Cubic>,<at 0.3125 1 RCIM_Cubic>,<at 1 0 RCIM_Cubic>]}}})
     2=?
     3=?
     Index=$Particles.VisibilityTag
     Num=::NewEnumerator0
    )
    Scale RGB=&MergeHelper_VectorSwitch(
     0=&VectorFromCurve(VectorCurve=DataInterface() {propertyValues {XCurve {keys [<at 0 0.1263>,<at 1 0.0096>]}, YCurve {keys [<at 0 0.4505>,<at 1 0.0505>]}, ZCurve {keys [<at 0 1>,<at 1 0.1424>]}}})
     1=&VectorFromCurve(VectorCurve=DataInterface() {propertyValues {
        XCurve {keys [{arriveTangent=-0.3505, interpMode=RCIM_Cubic, leaveTangent=-0.3505, tangentMode=RCTM_User, time=0.2102, value=1}, <at 0.7787 0 RCIM_Cubic>]}
        YCurve {keys [
          {arriveTangent=-0.961, interpMode=RCIM_Cubic, leaveTangent=-0.961, tangentMode=RCTM_User, time=0.2102, value=0.3945}
          <at 0.7787 0.6496 RCIM_Cubic>
         ]}
        ZCurve {keys [{interpMode=RCIM_Cubic, tangentMode=RCTM_User, time=0.2102, value=0.7393}, <at 0.7787 1 RCIM_Cubic>]}
       }})
     2=?
     3=?
     Index=$Particles.VisibilityTag
     Num=::NewEnumerator0
    )
    ScaleA=true
    ScaleRGB=true
   )
   &CameraOffset(Camera Offset Amount=&MergeHelper_FloatSwitch(0=20, 1=40, 2=?, 3=?, Index=$Particles.VisibilityTag, Num=::NewEnumerator0))
   &SubUVAnimation(
    End Frame Range Override=45
    "Sprite Renderer "=DataInterface() {propertyValues {SpriteRenderer=&self:Smoke_Local_0.NiagaraSpriteRendererProperties_1}}
    Start Frame Range Override=27
    UseEndFrame=true
    UseStartFrame=true
   )
   &AccelerationForce(Acceleration=<0 0 12>)
   &Drag(Use Linear Drag=true, Use Rotational Drag=true)
   &SolveForcesAndVelocity(Clamp Velocity=false, Limit Acceleration=false)
  ]
  rendererClasses [&NiagaraSpriteRendererProperties, &NiagaraMeshRendererProperties]
  renderers [
   {rendererClass=&NiagaraSpriteRendererProperties}
   {rendererClass=&NiagaraMeshRendererProperties, rendererIndex=1}
  ]
 }
 Smoke_World {
  emitterUpdateScript [
   &EmitterState(
    Life Cycle Mode=::NewEnumerator1
    Loop Duration=&RandomRangeFloat(Minimum=1.5, Maximum=3)
    UseLoopDelay=false
   )
   &SpawnPerUnit(
    Max Movement Threshold=1000
    Module.Use Max Movement Threshold=true
    Module.Use Movement Tolerance=true
    Movement Tolerance=0.1
    Spawn Spacing=10
    Use Spawn Probability=false
   )
  ]
  particleSpawnScript [
   &InitializeParticle(
    Color=<rgba 3 3 3 0.333>
    Lifetime=&RandomRangeFloat(Minimum=0.3, Maximum=0.5, Evaluation Type=::NewEnumerator1)
    Position Offset=&RandomRangeVector(Minimum=<0 12 55>, Maximum=<0 14 55>, Evaluation Type=::NewEnumerator1)
    Position Offset Coordinate Space=::Local
    Sprite Rotation Mode=::NewEnumerator2
    Sprite Size=&Vector2DFromFloat(Value=&RandomRangeFloat(Minimum=50, Maximum=70, Evaluation Type=::NewEnumerator1))
    Sprite Size Mode=::NewEnumerator4
    UsePositionOffset=true
    Write Material Random=true
   )
   &ShapeLocation(
    Cylinder Height=20
    Cylinder Radius=5
    MaskSpawnGroup=false
    Override Seed=false
    Rotation Mode=::NewEnumerator2
    Shape Primitive=::NewEnumerator1
    Surface Only=false
    UseOwnerScale=true
    Yaw / Pitch / Roll=<0 0 0.25>
   )
   &DynamicMaterialParameters(
    Index 0 Param 1=&RandomRangeFloat(Minimum=0, Maximum=1, Evaluation Type=::NewEnumerator1)
    Index 0 Param 2=&RandomRangeFloat(Minimum=1, Maximum=2, Evaluation Type=::NewEnumerator1)
    Index 0 Param 4=0.5
    Module.Index0Float0Write=true
    Module.Index0Float1Write=true
    Module.Index0Float2Write=true
    Module.Index0Float3Write=true
   )
   &AddVelocity(Velocity=&RandomRangeVector(Minimum=<0 -40 0>, Maximum=<0 -30 0>, Evaluation Type=::NewEnumerator1))
   &InheritVelocity(Inherited Velocity Amount Scale=<0.6 0.6 0.6>, Speed Limit=true, UseSpeedThreshold=false)
  ]
  particleUpdateScript [
   &ParticleState()
   &ScaleSpriteSize(
    Scale Sprite Size Mode=::NewEnumerator0
    Uniform Scale Factor=&FloatFromCurve(FloatCurve=DataInterface() {propertyValues {Curve {keys [{arriveTangent=0.3852, time=0, value=0.4456}, <at 0.9031 1.3>]}}})
   )
   &ScaleColor(
    Scale Alpha=&FloatFromCurve(FloatCurve=DataInterface() {propertyValues {Curve {keys [
         <at 0 0 RCIM_Cubic>
         {arriveTangent=0.3168, interpMode=RCIM_Cubic, leaveTangent=0.3168, time=0.3375, value=1}
         {arriveTangent=-1.5094, interpMode=RCIM_Cubic, leaveTangent=-1.5094, time=0.7386, value=0.234}
         <at 1 0 RCIM_Cubic>
        ]}}})
    Scale RGB=&VectorFromCurve(VectorCurve=DataInterface() {propertyValues {XCurve {keys [<at 0 0.1263>,<at 1 0.0096>]}, YCurve {keys [<at 0 0.4505>,<at 1 0.0505>]}, ZCurve {keys [<at 0 1>,<at 1 0.1424>]}}})
    ScaleA=true
    ScaleRGB=true
   )
   &CameraOffset(Camera Offset Amount=10)
   &SubUVAnimation(
    End Frame Range Override=42
    "Sprite Renderer "=DataInterface() {propertyValues {SpriteRenderer=&self:Smoke_World_0.NiagaraSpriteRendererProperties_1}}
    Start Frame Range Override=37
    UseEndFrame=true
    UseStartFrame=true
   )
   &AccelerationForce(Acceleration=<0 0 12>)
   &Drag(Use Linear Drag=true, Use Rotational Drag=true)
   &SolveForcesAndVelocity(Clamp Velocity=false, Limit Acceleration=false)
  ]
  rendererClasses [&NiagaraSpriteRendererProperties]
  renderers [{rendererClass=&NiagaraSpriteRendererProperties}]
 }
}
summary {
 systemName=NS_Pickaxe_Krampus_Idle
 emitters [
  {emitterName=Smoke_Local, rendererClasses [&NiagaraSpriteRendererProperties, &NiagaraMeshRendererProperties]}
  {emitterName=Smoke_World, rendererClasses [&NiagaraSpriteRendererProperties]}
  {emitterName=Light, rendererClasses [&NiagaraLightRendererProperties]}
 ]
 userVariables []
}
