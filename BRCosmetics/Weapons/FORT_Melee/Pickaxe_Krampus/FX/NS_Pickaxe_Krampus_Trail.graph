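# Niagara swing trail effect for the Krampus pickaxe, pairing an anim-trail ribbon with rising sprite particles.
# Main_Trail uses FN anim-trail modules to drive two ribbon renderers, with secondary narrower width and blue color attributes scaled by time of day.
# Particles_Swing spawns by rate and distance at a weapon socket, animating flipbook sprites that drag, drift upward and fade from magenta to blue.
# Sprites shrink in the front end.
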NiagaraSystem /BRCosmetics/Weapons/FORT_Melee/Pickaxe_Krampus/FX/NS_Pickaxe_Krampus_Trail
dependencies {
 usedDataInterfaces [
  &NiagaraDataInterfaceCurve
  &NiagaraDataInterfaceSpriteRendererInfo
  &NiagaraDataInterfaceVectorCurve
  &NiagaraDataInterfaceSocketReader
 ]
 usedDynamicInputs []
 usedModules [
  &self:SystemSpawnScript.SystemScriptSource.SystemScriptGraph.NiagaraNodeAssignment_0.SetVariables_#1
  &SystemState
  &self:SystemSpawnScript.SystemScriptSource.SystemScriptGraph.NiagaraNodeAssignment_1.SetVariables_#2
 ]
}
emitters {
 Main_Trail {
  emitterSpawnScript [&FN_RibbonIDOffsetUtility()]
  emitterUpdateScript [
   &EmitterState()
   &FN_SpawnAnimTrail()
  ]
  particleSpawnScript [
   &FN_InitializeAnimTrailBoneIndecies()
   &FN_InitializeAnimTrail(AnimTrailWidth=2, Color=<rgba 0.2 0.2 0.2 0.5>, Lifetime=0.35)
   &DynamicMaterialParameters(
    Index 0 Param 3=1
    Index 0 Param 4=&FloatFromCurve(
     FloatCurve=DataInterface() {propertyValues {Curve {keys [<at 0.0479 0 RCIM_Cubic>,<at 0.1 1 RCIM_Cubic>,<at 0.2 0 RCIM_Cubic>]}}}
     CurveIndex=$Emitter.LoopedAge
    )
    Module.Index0Float0Write=false
    Module.Index0Float1Write=false
    Module.Index0Float2Write=true
    Module.Index0Float3Write=true
   ) {module=DynamicMaterialParameters001}
   /SetVariables(Particles.Vector 4:Vector4f=&MakeVector4(X=1, Y=1, Z=1, W=1)) {isSetParameters=true}
  ]
  particleUpdateScript [
   &SolveForcesAndVelocity(Clamp Velocity=false, Limit Acceleration=false)
   &FN_UpdateAnimTrail() {module=FN_UpdateAnimTrail001}
   &ScaleColor(
    Scale Alpha=&Multiply_Float(
     A=&FloatFromCurve(FloatCurve=DataInterface() {propertyValues {Curve {keys [<at 0 0 RCIM_Cubic>,<at 0.2 1 RCIM_Cubic>,<at 1 0 RCIM_Cubic>]}}})
     B=$System.SystemFade
    )
    ScaleA=true
    ScaleRGB=true
   )
   &DynamicMaterialParameters(
    Index 0 Param 1=0.5
    Index 0 Param 2=-35
    Module.Index0Float0Write=true
    Module.Index0Float1Write=true
    Module.Index0Float2Write=false
    Module.Index0Float3Write=false
   )
   &ScaleRibbonWidth(Ribbon Width Scale=&FloatFromCurve(FloatCurve=DataInterface() {propertyValues {Curve {keys [
         {arriveTangent=0.1239, interpMode=RCIM_Cubic, leaveTangent=0.1239, tangentMode=RCTM_User, time=0, value=0.9627}
         {arriveTangent=-0.1002, interpMode=RCIM_Cubic, leaveTangent=-0.1002, tangentMode=RCTM_User, time=0.9983, value=0.857}
        ]}}}))
   /SetVariables(Particles.Color 02:LinearColor=&MakeLinearColorFromVectorAndFloat(
     "Vector (RGB)"=&VectorFromCurve(
      VectorCurve=DataInterface() {propertyValues {
        XCurve {keys [<at 0 0.1494>,<at 0.75 0>,<at 0.6 0>]}
        YCurve {keys [<at 0 0.8409>,<at 0.75 0.0889>,<at 0.6 0>]}
        ZCurve {keys [<at 0 3>,<at 0.75 1>,<at 0.6 0>]}
       }}
      Scale Curve=&VectorFromFloat(Value=&FNTimeOfDayScalar(Night Scale=1, Morning Scale=1.5, Day Scale=2.5, Evening Scale=1.2))
     )
     "Float (Alpha)"=&Multiply_Float(
      A=&FloatFromCurve(FloatCurve=DataInterface() {propertyValues {Curve {keys [<at 0 0 RCIM_Cubic>,<at 0.2 1 RCIM_Cubic>,<at 1 0 RCIM_Cubic>]}}}, Scale Curve=25)
      B=$System.SystemFade
     )
    )) {isSetParameters=true}
   /SetVariables(Particles.Position 02:NiagaraPosition=&AddVectorToPosition(Position=$Particles.Position, Vector=<0 0 0>)) {isSetParameters=true}
   /SetVariables(Particles.RibbonWidth 02:NiagaraFloat=&Multiply_Float(A=$Particles.RibbonWidth, B=0.6)) {isSetParameters=true}
  ]
  rendererClasses [&NiagaraRibbonRendererProperties]
  renderers [
   {rendererClass=&NiagaraRibbonRendererProperties}
   {rendererClass=&NiagaraRibbonRendererProperties, rendererIndex=1}
  ]
 }
 Particles_Swing {
  emitterUpdateScript [
   &EmitterState(Life Cycle Mode=::NewEnumerator1, Loop Behavior=::NewEnumerator1, Loop Delay=0.05, Loop Duration=0.3, UseLoopDelay=true)
   &SpawnRate(SpawnRate=15, Use Spawn Probability=false)
   &SpawnPerUnit(Module.Use Max Movement Threshold=true, Module.Use Movement Tolerance=false, Spawn Spacing=50, Use Spawn Probability=false)
  ]
  particleSpawnScript [
   &InitializeParticle(
    Color=<rgba 8 8 8 0.6>
    Lifetime Max=0.7
    Lifetime Min=0.4
    Lifetime Mode=::NewEnumerator1
    Sprite Rotation Mode=::NewEnumerator2
    Sprite Size Mode=::NewEnumerator3
    Uniform Sprite Size=&Multiply_Float(
     A=&RandomRangeFloat(Minimum=180, Maximum=100, Evaluation Type=::NewEnumerator1)
     B=&FN_FrontEndFloatModifier(Gameplay=1, Front End=0.6)
    )
    UsePositionOffset=false
    Write Material Random=true
   )
   &SocketLocation(
    Interpolate Socket Sampling=true
    SamplePosOffset=false
    Socket Index=1
    Socket Reader=$System.Socket Reader
    Socket Sampling Mode=::NewEnumerator3
    VelocitySpeedLimit=false
   )
   &InheritVelocity(Inherited Velocity Speed Limit=20, Speed Limit=true, UseSpeedThreshold=false)
   &DynamicMaterialParameters(
    Index 0 Param 1=&RandomRangeFloat(Minimum=0, Maximum=1)
    Index 0 Param 2=&RandomRangeFloat(Minimum=1, Maximum=2)
    Module.Index0Float0Write=true
    Module.Index0Float1Write=true
    Module.Index0Float2Write=false
    Module.Index0Float3Write=false
   )
  ]
  particleUpdateScript [
   &ParticleState()
   &ScaleSpriteSize(
    Scale Sprite Size Mode=::NewEnumerator0
    Uniform Scale Factor=&FloatFromCurve(FloatCurve=DataInterface() {propertyValues {Curve {keys [
         {arriveTangent=1.1101, interpMode=RCIM_Cubic, leaveTangent=1.1101, tangentMode=RCTM_User, time=0, value=0.4456}
         <at 0.9936 1.2297 RCIM_Cubic>
        ]}}})
   )
   &SpriteRotationRate(Rotation Rate=&RandomRangeFloat(Minimum=-1.5, Maximum=-0.5))
   &SubUVAnimation(
    End Frame Range Override=54
    "Sprite Renderer "=DataInterface() {propertyValues {SpriteRenderer=&self:Particles_Swing_0.NiagaraSpriteRendererProperties_1}}
    Start Frame Range Override=39
    UseEndFrame=true
    UseStartFrame=true
   )
   &Drag(Drag=5, Use Linear Drag=true, Use Rotational Drag=true)
   &AccelerationForce(Acceleration=<0 0 400>)
   &SolveForcesAndVelocity(Clamp Velocity=false, Limit Acceleration=false)
   &DynamicMaterialParameters(
    Index 0 Param 4=0.5
    Module.Index0Float0Write=false
    Module.Index0Float1Write=false
    Module.Index0Float2Write=true
    Module.Index0Float3Write=true
   ) {module=DynamicMaterialParameters001}
   &CameraOffset(Camera Offset Amount=-20)
   &ScaleColor(
    Scale Alpha=&FloatFromCurve(FloatCurve=DataInterface() {propertyValues {Curve {keys [
         <at 0 0.006 RCIM_Cubic>
         {arriveTangent=-0.006, interpMode=RCIM_Cubic, leaveTangent=-0.006, time=0.1938, value=0.1936}
         {arriveTangent=-0.7322, interpMode=RCIM_Cubic, leaveTangent=-0.7322, tangentMode=RCTM_User, time=1, value=0}
        ]}}})
    Scale RGB=&VectorFromCurve(VectorCurve=DataInterface() {propertyValues {
       XCurve {keys [<at -0.0009 1.0313 RCIM_Cubic>,<at 0.8489 0.0548 RCIM_Cubic>]}
       YCurve {keys [<at -0.0009 0.0783 RCIM_Cubic>,<at 0.8489 0.3412 RCIM_Cubic>]}
       ZCurve {keys [<at -0.0009 0.8428 RCIM_Cubic>,<at 0.8489 0.7379 RCIM_Cubic>]}
      }})
    ScaleA=true
    ScaleRGB=true
   )
  ]
  rendererClasses [&NiagaraSpriteRendererProperties]
  renderers [{rendererClass=&NiagaraSpriteRendererProperties}]
 }
}
summary {
 systemName=NS_Pickaxe_Krampus_Trail
 emitters [
  {emitterName=Particles_Swing, rendererClasses [&NiagaraSpriteRendererProperties]}
  {emitterName=Main_Trail, rendererClasses [&NiagaraRibbonRendererProperties]}
 ]
 userVariables [
  User.bIsFrontEnd {type=NiagaraBool}
  User.bIsFrontEndPreview {type=NiagaraBool}
  User.RibbonIDOffset {type=NiagaraInt32}
 ]
}
