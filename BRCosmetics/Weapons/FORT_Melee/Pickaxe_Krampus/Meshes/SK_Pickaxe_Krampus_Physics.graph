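# Physics asset for the Krampus pickaxe skeletal mesh, defining capsule bodies and joints for its dangling parts.
# A long handle capsule anchors two branches, a four-link chain and upper and lower ribbon strands, each built as constrained bone chains.
# Branches are tightly limited in swing and twist, chain links swing but barely twist, and ribbons bend and twist more freely.
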PhysicsAsset /BRCosmetics/Weapons/FORT_Melee/Pickaxe_Krampus/Meshes/SK_Pickaxe_Krampus_Physics
bodies {
 Melee_TwoHanded_Handle [{shapeType=Capsule, length=94.75, radius=4, center=<1.201 -0.9369 26.6801>, rotation=<rot 0 0 0.0005>}]
 dyn_branch_1 [{shapeType=Capsule, length=15, radius=7, center=<1 0 0>, rotation=<rot 90 0 0>}]
 dyn_branch_2 [{shapeType=Capsule, length=15, radius=7, center=<1 0 0>, rotation=<rot 90 0 0>}]
 dyn_chain_1 [{shapeType=Capsule, length=4, radius=3.5, center=<2 0 0>, rotation=<rot 90 0 0>}]
 dyn_chain_2 [{shapeType=Capsule, length=4, radius=3.5, center=<2 0 0>, rotation=<rot 90 0 0>}]
 dyn_chain_3 [{shapeType=Capsule, length=4, radius=3.5, center=<2 0 0>, rotation=<rot 90 0 0>}]
 dyn_chain_4 [{shapeType=Capsule, length=4, radius=3.5, center=<2 0 0>, rotation=<rot 90 0 0>}]
 dyn_ribbon_dn_1 [{shapeType=Capsule, length=6, radius=1.5, center=<0 -3 0>, rotation=<rot 0 0 90>}]
 dyn_ribbon_dn_2 [{shapeType=Capsule, length=6, radius=1.5, center=<0 -3 0>, rotation=<rot 0 0 90>}]
 dyn_ribbon_dn_3 [{shapeType=Capsule, length=6, radius=1.5, center=<0 -3 0>, rotation=<rot 0 0 90>}]
 dyn_ribbon_dn_4 [{shapeType=Capsule, length=5, radius=3, center=<0 -2 0>, rotation=<rot 0 0 90>}]
 dyn_ribbon_up_1 [{shapeType=Capsule, length=6, radius=1.5, center=<0 -3 0>, rotation=<rot 0 0 90>}]
 dyn_ribbon_up_2 [{shapeType=Capsule, length=6, radius=1.5, center=<0 -3 0>, rotation=<rot 0 0 90>}]
 dyn_ribbon_up_3 [{shapeType=Capsule, length=5, radius=3, center=<0 -2 0>, rotation=<rot 0 0 90>}]
}
constraints [
 {bone1Name=dyn_branch_1, bone2Name=Melee_TwoHanded_Handle, swing1LimitDegrees=35, swing2LimitDegrees=10, twistLimitDegrees=2}
 {bone1Name=dyn_branch_2, bone2Name=Melee_TwoHanded_Handle, swing1LimitDegrees=35, swing2LimitDegrees=10, twistLimitDegrees=2}
 {bone1Name=dyn_ribbon_up_1, bone2Name=Melee_TwoHanded_Handle, swing1LimitDegrees=45, swing2LimitDegrees=45, twistLimitDegrees=45}
 {bone1Name=dyn_ribbon_up_2, bone2Name=dyn_ribbon_up_1, swing1LimitDegrees=45, swing2LimitDegrees=45, twistLimitDegrees=45}
 {bone1Name=dyn_ribbon_up_3, bone2Name=dyn_ribbon_up_2, swing1LimitDegrees=45, swing2LimitDegrees=45, twistLimitDegrees=45}
 {bone1Name=dyn_ribbon_dn_1, bone2Name=Melee_TwoHanded_Handle, swing1LimitDegrees=45, swing2LimitDegrees=45, twistLimitDegrees=45}
 {bone1Name=dyn_ribbon_dn_2, bone2Name=dyn_ribbon_dn_1, swing1LimitDegrees=45, swing2LimitDegrees=45, twistLimitDegrees=45}
 {bone1Name=dyn_ribbon_dn_3, bone2Name=dyn_ribbon_dn_2, swing1LimitDegrees=45, swing2LimitDegrees=45, twistLimitDegrees=45}
 {bone1Name=dyn_ribbon_dn_4, bone2Name=dyn_ribbon_dn_3, swing1LimitDegrees=45, swing2LimitDegrees=45, twistLimitDegrees=45}
 {bone1Name=dyn_chain_1, bone2Name=Melee_TwoHanded_Handle, swing1LimitDegrees=45, swing2LimitDegrees=15, twistLimitDegrees=3}
 {bone1Name=dyn_chain_2, bone2Name=dyn_chain_1, swing1LimitDegrees=45, swing2LimitDegrees=45, twistLimitDegrees=3}
 {bone1Name=dyn_chain_3, bone2Name=dyn_chain_2, swing1LimitDegrees=45, swing2LimitDegrees=45, twistLimitDegrees=3}
 {bone1Name=dyn_chain_4, bone2Name=dyn_chain_3, swing1LimitDegrees=45, swing2LimitDegrees=45, twistLimitDegrees=3}
]
