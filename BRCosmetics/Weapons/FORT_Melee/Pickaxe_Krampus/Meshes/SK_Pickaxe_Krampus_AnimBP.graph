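# Animation Blueprint for the Krampus pickaxe skeleton that adds secondary motion to its dangling chain and ribbons.
# The cached input pose feeds two paths, blended by bRigidBodyEnabled: a RigidBody simulation in handle bone space, or additive bone rotations followed by Trail controllers.
# An Anim Dynamics stack for chain and branch bones and a single-state branch-animation state machine are not connected to the output.
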AnimBlueprint /BRCosmetics/Weapons/FORT_Melee/Pickaxe_Krampus/Meshes/SK_Pickaxe_Krampus_AnimBP
target_skeleton=&/SK_Pickaxe_Krampus_Skeleton
graphs [
 AnimGraph {nodes [
   n1 = Anim Dynamics - Bone: dyn_chain_1(Alpha=1, ExternalForce="0,0,0") {
    angularDampingOverride=0.999
    angularSpringConstant=1000
    bAngularSpring=true
    bGravityOverrideInSimSpace=true
    bOverrideAngularDamping=true
    bOverrideLinearDamping=true
    linearDampingOverride=0.999
    alphaBoolBlend {blendOption=Linear}
    boundBone {boneName=dyn_chain_1}
    componentAppliedLinearAccClamp=<1000 1000 1000>
    componentLinearAccScale=<1 1 1>
    componentLinearVelScale=<1 1 1>
    gravityOverride {z=-980}
    physicsBodyDefinitions [{
      sphereCollisionRadius=0
      boundBone {boneName=dyn_chain_1}
      boxExtents=<5 5 5>
      constraintSetup {angularTargetAxis=AxisY, angularLimitsMax=<1 10 90>, angularLimitsMin=<-1 -10 -90>, angularTarget=<0 0 0>}
      localJointOffset=<1 -0.1 -0.1>
     }]
   }
   n2 = Anim Dynamics - Bone: dyn_chain_2(ComponentPose=@n1, Alpha=1, ExternalForce="0,0,0") {
    angularDampingOverride=0.999
    angularSpringConstant=1000
    bAngularSpring=true
    bGravityOverrideInSimSpace=true
    bOverrideAngularDamping=true
    bOverrideLinearDamping=true
    linearDampingOverride=0.999
    alphaBoolBlend {blendOption=Linear}
    boundBone {boneName=dyn_chain_2}
    componentAppliedLinearAccClamp=<1000 1000 1000>
    componentLinearAccScale=<1 1 1>
    componentLinearVelScale=<1 1 1>
    gravityOverride {z=-980}
    physicsBodyDefinitions [{
      sphereCollisionRadius=0
      boundBone {boneName=dyn_chain_2}
      boxExtents=<20 20 20>
      constraintSetup {angularTargetAxis=AxisY, angularLimitsMax=<0 80 90>, angularLimitsMin=<0 -80 -90>, angularTarget=<0 0 0>}
      localJointOffset=<12 -0.2 0.1>
     }]
   }
   n3 = Anim Dynamics - Bone: dyn_chain_3(ComponentPose=@n2, Alpha=1, ExternalForce="0,0,0") {
    angularDampingOverride=0.999
    angularSpringConstant=1000
    bAngularSpring=true
    bGravityOverrideInSimSpace=true
    bOverrideAngularDamping=true
    bOverrideLinearDamping=true
    gravityScale=0.5
    linearDampingOverride=0.999
    alphaBoolBlend {blendOption=Linear}
    boundBone {boneName=dyn_chain_3}
    componentAppliedLinearAccClamp=<1000 1000 1000>
    componentLinearAccScale=<1 1 1>
    componentLinearVelScale=<1 1 1>
    gravityOverride {z=-980}
    physicsBodyDefinitions [{
      sphereCollisionRadius=0
      boundBone {boneName=dyn_chain_3}
      boxExtents=<20 20 20>
      constraintSetup {angularTargetAxis=AxisY, angularLimitsMax=<0 20 180>, angularLimitsMin=<0 -20 -180>, angularTarget=<0 0 0>}
      localJointOffset=<14 -0.1 -0.1>
     }]
   }
   n4 = Anim Dynamics - Bone: dyn_chain_4(ComponentPose=@n3, Alpha=1, ExternalForce="0,0,0") {
    angularDampingOverride=0.999
    angularSpringConstant=1000
    bAngularSpring=true
    bGravityOverrideInSimSpace=true
    bOverrideAngularDamping=true
    bOverrideLinearDamping=true
    gravityScale=0.3
    linearDampingOverride=0.999
    alphaBoolBlend {blendOption=Linear}
    boundBone {boneName=dyn_chain_4}
    componentAppliedLinearAccClamp=<1000 1000 1000>
    componentLinearAccScale=<1 1 1>
    componentLinearVelScale=<1 1 1>
    gravityOverride {z=-980}
    physicsBodyDefinitions [{
      sphereCollisionRadius=0
      boundBone {boneName=dyn_chain_4}
      boxExtents=<20 20 20>
      constraintSetup {angularTargetAxis=AxisY, angularLimitsMax=<0 80 90>, angularLimitsMin=<0 -80 -90>, angularTarget=<0 0 0>}
      localJointOffset=<16 0.1 0.1>
     }]
   }
   n5 = Anim Dynamics - Bone: dyn_branch_1(ComponentPose=@n4, Alpha=1, ExternalForce="0,0,0") {
    angularDampingOverride=0.999
    angularSpringConstant=3000
    bAngularSpring=true
    bGravityOverrideInSimSpace=true
    bOverrideAngularDamping=true
    bOverrideLinearDamping=true
    bUseGravityOverride=true
    linearDampingOverride=0.999
    alphaBoolBlend {blendOption=Linear}
    boundBone {boneName=dyn_branch_1}
    componentAppliedLinearAccClamp=<1000000 1000000 1000000>
    componentLinearAccScale=<1 1 1>
    componentLinearVelScale=<1 1 1>
    gravityOverride {z=-980}
    physicsBodyDefinitions [{
      sphereCollisionRadius=0
      boundBone {boneName=dyn_branch_1}
      boxExtents=<20 20 20>
      constraintSetup {angularLimitsMax=<1 1 1>, angularLimitsMin=<-1 -1 -1>, angularTarget=<0 0 0>}
      localJointOffset=<1 0 0>
     }]
   }
   n6 = Anim Dynamics - Bone: dyn_branch_1_1(ComponentPose=@n5, Alpha=1, ExternalForce="0,0,0") {
    angularDampingOverride=0.999
    angularSpringConstant=3000
    bAngularSpring=true
    bGravityOverrideInSimSpace=true
    bOverrideAngularDamping=true
    bOverrideLinearDamping=true
    bUseGravityOverride=true
    linearDampingOverride=0.999
    alphaBoolBlend {blendOption=Linear}
    boundBone {boneName=dyn_branch_1_1}
    componentLinearAccScale=<1 1 1>
    componentLinearVelScale=<1 1 1>
    gravityOverride {z=-980}
    physicsBodyDefinitions [{
      sphereCollisionRadius=0
      boundBone {boneName=dyn_branch_1_1}
      boxExtents=<20 20 20>
      constraintSetup {angularLimitsMax=<1 1 1>, angularLimitsMin=<-1 -1 -1>, angularTarget=<0 0 0>}
      localJointOffset=<3 0 0>
     }]
   }
   n7 = Anim Dynamics - Bone: dyn_branch_1_2(ComponentPose=@n6, Alpha=1, ExternalForce="0,0,0") {
    angularDampingOverride=0.999
    angularSpringConstant=3000
    bAngularSpring=true
    bGravityOverrideInSimSpace=true
    bOverrideAngularDamping=true
    bOverrideLinearDamping=true
    bUseGravityOverride=true
    linearDampingOverride=0.999
    alphaBoolBlend {blendOption=Linear}
    boundBone {boneName=dyn_branch_1_2}
    componentLinearAccScale=<1 1 1>
    componentLinearVelScale=<1 1 1>
    gravityOverride {z=-980}
    physicsBodyDefinitions [{
      sphereCollisionRadius=0
      boundBone {boneName=dyn_branch_1_2}
      boxExtents=<20 20 20>
      constraintSetup {angularLimitsMax=<1 1 1>, angularLimitsMin=<-1 -1 -1>, angularTarget=<0 0 0>}
      localJointOffset=<3 0 0>
     }]
   }
   n8 = Anim Dynamics - Bone: dyn_branch_1_3(ComponentPose=@n7, Alpha=1, ExternalForce="0,0,0") {
    angularDampingOverride=0.999
    angularSpringConstant=3000
    bAngularSpring=true
    bGravityOverrideInSimSpace=true
    bOverrideAngularDamping=true
    bOverrideLinearDamping=true
    bUseGravityOverride=true
    linearDampingOverride=0.999
    alphaBoolBlend {blendOption=Linear}
    boundBone {boneName=dyn_branch_1_3}
    componentLinearAccScale=<1 1 1>
    componentLinearVelScale=<1 1 1>
    gravityOverride {z=-980}
    physicsBodyDefinitions [{
      sphereCollisionRadius=0
      boundBone {boneName=dyn_branch_1_3}
      boxExtents=<20 20 20>
      constraintSetup {angularLimitsMax=<1 1 1>, angularLimitsMin=<-1 -1 -1>, angularTarget=<0 0 0>}
      localJointOffset=<3 0 0>
     }]
   }
   n9 = Anim Dynamics - Bone: dyn_branch_2(ComponentPose=@n8, Alpha=1, ExternalForce="0,0,0") {
    angularDampingOverride=0.999
    angularSpringConstant=3000
    bAngularSpring=true
    bGravityOverrideInSimSpace=true
    bOverrideAngularDamping=true
    bOverrideLinearDamping=true
    bUseGravityOverride=true
    linearDampingOverride=0.999
    alphaBoolBlend {blendOption=Linear}
    boundBone {boneName=dyn_branch_2}
    componentAppliedLinearAccClamp=<1000000 1000000 1000000>
    componentLinearAccScale=<1 1 1>
    componentLinearVelScale=<1 1 1>
    gravityOverride {z=-980}
    physicsBodyDefinitions [{
      sphereCollisionRadius=0
      boundBone {boneName=dyn_branch_2}
      boxExtents=<20 20 20>
      constraintSetup {angularLimitsMax=<1 1 1>, angularLimitsMin=<-1 -1 -1>, angularTarget=<0 0 0>}
      localJointOffset=<1 -0.1 0.1>
     }]
   }
   n10 = Anim Dynamics - Bone: dyn_branch_2_1(ComponentPose=@n9, Alpha=1, ExternalForce="0,0,0") {
    angularDampingOverride=0.999
    angularSpringConstant=3000
    bAngularSpring=true
    bGravityOverrideInSimSpace=true
    bOverrideAngularDamping=true
    bOverrideLinearDamping=true
    bUseGravityOverride=true
    linearDampingOverride=0.999
    alphaBoolBlend {blendOption=Linear}
    boundBone {boneName=dyn_branch_2_1}
    componentLinearAccScale=<1 1 1>
    componentLinearVelScale=<1 1 1>
    gravityOverride {z=-980}
    physicsBodyDefinitions [{
      sphereCollisionRadius=0
      boundBone {boneName=dyn_branch_2_1}
      boxExtents=<20 20 20>
      constraintSetup {angularLimitsMax=<1 1 1>, angularLimitsMin=<-1 -1 -1>, angularTarget=<0 0 0>}
      localJointOffset=<3 -0.1 0.1>
     }]
   }
   Anim Dynamics - Bone: dyn_branch_2_2(ComponentPose=@n10, Alpha=1, ExternalForce="0,0,0") {
    angularDampingOverride=0.999
    angularSpringConstant=3000
    bAngularSpring=true
    bGravityOverrideInSimSpace=true
    bOverrideAngularDamping=true
    bOverrideLinearDamping=true
    bUseGravityOverride=true
    linearDampingOverride=0.999
    alphaBoolBlend {blendOption=Linear}
    boundBone {boneName=dyn_branch_2_2}
    componentLinearAccScale=<1 1 1>
    componentLinearVelScale=<1 1 1>
    gravityOverride {z=-980}
    physicsBodyDefinitions [{
      sphereCollisionRadius=0
      boundBone {boneName=dyn_branch_2_2}
      boxExtents=<20 20 20>
      constraintSetup {angularLimitsMax=<1 1 1>, angularLimitsMin=<-1 -1 -1>, angularTarget=<0 0 0>}
      localJointOffset=<3 -0.1 0.1>
     }]
   }
   n11 = "Transform (Modify) Bone - Bone: dyn_chain_1"(ComponentPose=Local To Component(LocalPose=Use cached pose 'In Pose'()), Alpha=1, Translation="0,0,0", Rotation="-20,0,0", Scale="1,1,1") {
    rotationMode=BMM_Additive
    rotationSpace=BCS_BoneSpace
    alphaBoolBlend {blendOption=Linear}
    boneToModify {boneName=dyn_chain_1}
    rotation {pitch=-20}
   }
   n12 = "Transform (Modify) Bone - Bone: dyn_chain_2"(ComponentPose=@n11, Alpha=1, Translation="0,0,0", Rotation="-40,-30,0", Scale="1,1,1") {
    rotationMode=BMM_Additive
    rotationSpace=BCS_BoneSpace
    alphaBoolBlend {blendOption=Linear}
    boneToModify {boneName=dyn_chain_2}
    rotation {pitch=-40, yaw=-30}
   }
   n13 = "Transform (Modify) Bone - Bone: dyn_ribbon_up_1"(ComponentPose=@n12, Alpha=1, Translation="0,0,0", Rotation="-20,0.000,-60", Scale="1,1,1") {
    rotationMode=BMM_Additive
    rotationSpace=BCS_BoneSpace
    alphaBoolBlend {blendOption=Linear}
    boneToModify {boneName=dyn_ribbon_up_1}
    rotation {pitch=-20, roll=-60}
   }
   n14 = "Transform (Modify) Bone - Bone: dyn_ribbon_dn_1"(ComponentPose=@n13, Alpha=1, Translation="0,0,0", Rotation="-20,0,-60", Scale="1,1,1") {
    rotationMode=BMM_Additive
    rotationSpace=BCS_BoneSpace
    alphaBoolBlend {blendOption=Linear}
    boneToModify {boneName=dyn_ribbon_dn_1}
    rotation {pitch=-20, roll=-60}
   }
   n15 = Trail controller - Bone: dyn_chain_4(ComponentPose=@n14, Alpha=1) {
    chainLength=3
    relaxationSpeedScale=2
    rotationLimits [{limitMax=<180 180 180>}, {limitMax=<180 180 180>}, {limitMax=<180 180 180>}]
    rotationOffsets [<0 0 0>, <0 0 0>, <0 0 0>]
    trailBone {boneName=dyn_chain_4}
   }
   n16 = Trail controller - Bone: dyn_ribbon_up_3(ComponentPose=@n15, Alpha=1) {
    bInvertChainBoneAxis=true
    chainBoneAxis=Y
    chainLength=3
    relaxationSpeedScale=2
    rotationLimits [{limitMax=<180 180 180>}, {limitMax=<180 180 180>}, {limitMax=<180 180 180>}]
    rotationOffsets [<0 0 0>, <0 0 0>, <0 0 0>]
    trailBone {boneName=dyn_ribbon_up_3}
   }
   n17 = Trail controller - Bone: dyn_ribbon_dn_4(ComponentPose=@n16, Alpha=1) {
    bInvertChainBoneAxis=true
    chainBoneAxis=Y
    chainLength=4
    relaxationSpeedScale=2
    rotationLimits [{limitMax=<180 180 180>}, {limitMax=<180 180 180>}, {limitMax=<180 180 180>}, {limitMax=<180 180 180>}]
    rotationOffsets [<0 0 0>, <0 0 0>, <0 0 0>, <0 0 0>]
    trailBone {boneName=dyn_ribbon_dn_4}
   }
   n18 = RigidBody(ComponentPose=Local To Component(LocalPose=Use cached pose 'In Pose'())) {
    bClampLinearTranslationLimitToRefPose=true
    bOverrideWorldGravity=true
    overridePhysicsAsset=&/SK_Pickaxe_Krampus_Physics
    simulationSpace=BaseBoneSpace
    alphaBoolBlend {blendOption=Linear}
    baseBoneRef {boneName=Melee_TwoHanded_Handle}
    componentAppliedLinearAccClamp=<3000 3000 3000>
    componentLinearAccScale=<0.35 0.45 0.45>
    componentLinearVelScale=<0.3 0.3 0.3>
    overrideWorldGravity {z=-980}
   }
   Output Pose(Result=Blend Poses by bool(BlendPose_0=Component To Local(ComponentPose=@n18), BlendPose_1=Component To Local(ComponentPose=@n17), bActiveValue=Get bRigidBodyEnabled()) {blendType=Linear}) {layerGroup=None}
   Save cached pose 'In Pose'(Pose=Input Pose()) {cacheName=In Pose}
   AnimGraphNode_StateMachine_1 = New State Machine() {is=title AnimGraphNode_StateMachine}
  ]}
 New State Machine {entry=Base, path=AnimGraph.AnimGraphNode_StateMachine_1.New State Machine, ids=id, nodes [AnimStateNode_0 = Base() {is=title AnimStateNode}]}
 Base {
  path=AnimGraph.AnimGraphNode_StateMachine_1.New State Machine.AnimStateNode_0.Base
  nodes [
   n1 = Sequence Player 'Anim_Pickaxe_Krampus_branch'() {sequence=/BRCosmetics/Weapons/FORT_Melee/Pickaxe_Krampus/Animation/Anim_Pickaxe_Krampus_branch, playRateScaleBiasClampConstants {scale=0.1}}
   Output Animation Pose(Result=@n1)
  ]
 }
 EventGraph {nodes [Event BlueprintUpdateAnimation()]}
]
variables []
